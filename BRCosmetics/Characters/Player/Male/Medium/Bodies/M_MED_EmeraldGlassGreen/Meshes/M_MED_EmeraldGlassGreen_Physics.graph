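# Physics asset for the male medium player body M_MED_EmeraldGlassGreen, covering only the upper torso, upper arms and the dyn_mask_main bone.
# Capsules form the bodies on spine_05, both clavicles and both upper arms, while dyn_mask_main carries three capsules, two of them mirrored side to side.
# Constraints chain each upper arm through its clavicle to spine_05 with free swing and twist, and join dyn_mask_main directly to spine_05.
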PhysicsAsset /BRCosmetics/Characters/Player/Male/Medium/Bodies/M_MED_EmeraldGlassGreen/Meshes/M_MED_EmeraldGlassGreen_Physics
bodies {
 clavicle_l [{
   shapeType=Capsule
   length=10
   radius=12.1
   shapeName=clavicle_l_capsule
   center=<8.6336 -1.252 -6.7199>
   rotation=<rot 84.5161 5.6017 16.9692>
  }]
 clavicle_r [{
   shapeType=Capsule
   length=10
   radius=12.1
   shapeName=clavicle_l_capsule
   center=<-8.6336 1.252 6.7199>
   rotation=<rot 84.5161 5.6018 16.9693>
  }]
 dyn_mask_main [
  {
   shapeType=Capsule
   length=10.3404
   radius=3.5598
   shapeName=dyn_mask_main_capsule
   center=<2.9712 -6 2.6175>
   rotation=<rot 77.1945 -180 -180>
  }
  {
   shapeType=Capsule
   length=10.3404
   radius=3.5598
   shapeName=dyn_mask_main_capsule
   center=<2.9712 6 2.6175>
   rotation=<rot 77.1945 -180 -180>
  }
  {shapeType=Capsule, length=2.1568, radius=4, shapeName=dyn_mask_main_capsule, center=<15.1228 0 4.0284>, rotation=<rot 0 0 90>}
 ]
 spine_05 [{shapeType=Capsule, length=10, radius=10.3, shapeName=spine_05_capsule, center=<-3.3788 0.8822 0>}]
 upperarm_l [{shapeType=Capsule, length=25, radius=3.5, shapeName=upperarm_r_capsule, center=<10.4395 0 -0.8588>, rotation=<rot 90 0 0>}]
 upperarm_r [{shapeType=Capsule, length=25, radius=3.5, shapeName=upperarm_r_capsule, center=<-10.4395 0 0.8588>, rotation=<rot 90 0 0>}]
}
constraints [
 {bone1Name=dyn_mask_main, bone2Name=spine_05, swing1LimitDegrees=45, swing2LimitDegrees=45, twistLimitDegrees=45}
 {
  bone1Name=upperarm_l
  bone2Name=clavicle_l
  swing1LimitDegrees=45
  swing1Motion=Free
  swing2LimitDegrees=45
  swing2Motion=Free
  twistLimitDegrees=45
  twistMotion=Free
 }
 {
  bone1Name=upperarm_r
  bone2Name=clavicle_r
  swing1LimitDegrees=45
  swing1Motion=Free
  swing2LimitDegrees=45
  swing2Motion=Free
  twistLimitDegrees=45
  twistMotion=Free
 }
 {
  bone1Name=clavicle_l
  bone2Name=spine_05
  swing1LimitDegrees=45
  swing1Motion=Free
  swing2LimitDegrees=45
  swing2Motion=Free
  twistLimitDegrees=45
  twistMotion=Free
 }
 {
  bone1Name=clavicle_r
  bone2Name=spine_05
  swing1LimitDegrees=45
  swing1Motion=Free
  swing2LimitDegrees=45
  swing2Motion=Free
  twistLimitDegrees=45
  twistMotion=Free
 }
]
