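# Physics asset for the hood of the EmeraldGlassGreen medium male player body's cowl, built entirely from capsule bodies.
# Left and right four-bone dyn hood chains hang from spine_05, cross-linked at their first bones, with twist locked on the last two joints of each.
# Upper spine and clavicle bodies form the torso, joined by free-rotation constraints, with spine_04 adding two large offset side capsules.
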PhysicsAsset /BRCosmetics/Characters/Player/Male/Medium/Bodies/M_MED_EmeraldGlassGreen/Meshes/M_MED_EmeraldGlassGreen_Physics_Cowl
bodies {
 clavicle_l [{shapeType=Capsule, length=10, radius=8.5, shapeName=clavicle_l_capsule, center=<8.3043 -0.8728 -2.1072>, rotation=<rot 85 -0.0005 -0.0005>}]
 clavicle_r [{shapeType=Capsule, length=10, radius=8.5, shapeName=clavicle_l_capsule, center=<-8.3043 0.8728 2.1072>, rotation=<rot 85 -0.0005 -0.0005>}]
 dyn_L_hood_1 [{shapeType=Capsule, length=4, radius=2.7, shapeName=dyn_L_hood_1_capsule, center=<4 0 0>, rotation=<rot 90 0 0>}]
 dyn_L_hood_2 [{shapeType=Capsule, length=4, radius=3.2, shapeName=dyn_L_hood_2_capsule, center=<4 0 1.4>, rotation=<rot 90 0 0>}]
 dyn_L_hood_3 [{shapeType=Capsule, length=4, radius=3.2, shapeName=dyn_L_hood_3_capsule, center=<4 0 0>, rotation=<rot 90 0 0>}]
 dyn_L_hood_4 [{shapeType=Capsule, length=4, radius=3.2, shapeName=dyn_L_hood_4_capsule, center=<4 0 0>, rotation=<rot 90 0 0>}]
 dyn_R_hood_1 [{shapeType=Capsule, length=4, radius=2.7, shapeName=dyn_R_hood_1_capsule, center=<4 0 0>, rotation=<rot 90 0 0>}]
 dyn_R_hood_2 [{shapeType=Capsule, length=4, radius=3.2, shapeName=dyn_R_hood_2_capsule, center=<4 0 -1.4>, rotation=<rot 90 0 0>}]
 dyn_R_hood_3 [{shapeType=Capsule, length=4, radius=3.2, shapeName=dyn_R_hood_3_capsule, center=<4 0 0>, rotation=<rot 90 0 0>}]
 dyn_R_hood_4 [{shapeType=Capsule, length=4, radius=3.2, shapeName=dyn_R_hood_4_capsule, center=<4 0 0>, rotation=<rot 90 0 0>}]
 spine_03 [{shapeType=Capsule, length=20, radius=10, shapeName=spine_03_capsule, center=<0 1.0582 0>}]
 spine_04 [
  {shapeType=Capsule, length=20, radius=10, shapeName=spine_04_capsule, center=<0 1.772 0>}
  {shapeType=Capsule, length=30, radius=20, shapeName=spine_04_capsule, center=<0 -15 -32>, rotation=<rot 90 0 0>}
  {shapeType=Capsule, length=30, radius=20, shapeName=spine_04_capsule, center=<0 -15 32>, rotation=<rot 90 0 0>}
 ]
 spine_05 [{shapeType=Capsule, length=10, radius=10.3, shapeName=spine_05_capsule, center=<-3.4081 0.78 0>}]
}
constraints [
 {
  bone1Name=spine_04
  bone2Name=spine_03
  swing1LimitDegrees=45
  swing1Motion=Free
  swing2LimitDegrees=45
  swing2Motion=Free
  twistLimitDegrees=45
  twistMotion=Free
 }
 {
  bone1Name=spine_05
  bone2Name=spine_04
  swing1LimitDegrees=45
  swing1Motion=Free
  swing2LimitDegrees=45
  swing2Motion=Free
  twistLimitDegrees=45
  twistMotion=Free
 }
 {bone1Name=dyn_L_hood_1, bone2Name=spine_05, swing1LimitDegrees=45, swing2LimitDegrees=45, twistLimitDegrees=45}
 {bone1Name=dyn_L_hood_2, bone2Name=dyn_L_hood_1, swing1LimitDegrees=30, swing2LimitDegrees=45, twistLimitDegrees=5}
 {bone1Name=dyn_L_hood_3, bone2Name=dyn_L_hood_2, swing1LimitDegrees=20, swing2LimitDegrees=45, twistLimitDegrees=45, twistMotion=Locked}
 {bone1Name=dyn_L_hood_4, bone2Name=dyn_L_hood_3, swing1LimitDegrees=20, swing2LimitDegrees=45, twistLimitDegrees=45, twistMotion=Locked}
 {bone1Name=dyn_R_hood_1, bone2Name=spine_05, swing1LimitDegrees=45, swing2LimitDegrees=45, twistLimitDegrees=45}
 {bone1Name=dyn_R_hood_2, bone2Name=dyn_R_hood_1, swing1LimitDegrees=30, swing2LimitDegrees=45, twistLimitDegrees=5}
 {bone1Name=dyn_R_hood_3, bone2Name=dyn_R_hood_2, swing1LimitDegrees=20, swing2LimitDegrees=45, twistLimitDegrees=45, twistMotion=Locked}
 {bone1Name=dyn_R_hood_4, bone2Name=dyn_R_hood_3, swing1LimitDegrees=20, swing2LimitDegrees=45, twistLimitDegrees=45, twistMotion=Locked}
 {bone1Name=dyn_L_hood_1, bone2Name=dyn_R_hood_1, swing1LimitDegrees=60, swing2LimitDegrees=60, twistLimitDegrees=5}
 {
  bone1Name=clavicle_l
  bone2Name=spine_05
  swing1LimitDegrees=45
  swing1Motion=Free
  swing2LimitDegrees=45
  swing2Motion=Free
  twistLimitDegrees=45
  twistMotion=Free
 }
 {
  bone1Name=clavicle_r
  bone2Name=spine_05
  swing1LimitDegrees=45
  swing1Motion=Free
  swing2LimitDegrees=45
  swing2Motion=Free
  twistLimitDegrees=45
  twistMotion=Free
 }
]
