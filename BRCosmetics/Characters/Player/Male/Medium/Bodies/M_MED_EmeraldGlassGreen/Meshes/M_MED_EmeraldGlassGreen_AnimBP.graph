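# Character-part AnimBP for the EmeraldGlassGreen male body, adding secondary motion and corrective bone adjustments to the pose copied from its parent mesh.
# RigidBody simulates mask and cowl physics, while AnimDynamics swings laces and pockets.
# Constraints and additive transforms correct mask, knees, wrists, pockets and backpack attachment.
# The event graph drives these from joint rotations, backbling state (cowl sim off, heavier gravity) and tagged montage playback outside emotes.
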AnimBlueprint /BRCosmetics/Characters/Player/Male/Medium/Bodies/M_MED_EmeraldGlassGreen/Meshes/M_MED_EmeraldGlassGreen_AnimBP
target_skeleton=&/M_MED_EmeraldGlassGreen_Skeleton
graphs [
 AnimGraph {nodes [
   n1 = "Transform (Modify) Bone - Bone: attach_backpack"!(
    ComponentPose=Local To Component(LocalPose=Copy Pose From Mesh() {bCopyCurves=true, bUseAttachedParent=true})
    bAlphaBoolEnabled:bool=Get notCape() -> (notCape:bool)
    Translation="0.000,-0.7,0.000"
    Rotation="0,0,0"
    Scale="1,1,1"
   ) -> (Pose) {
    alphaInputType=Bool
    translationMode=BMM_Additive
    alphaBoolBlend {blendOption=Linear}
    boneToModify {boneName=attach_backpack}
    translation {y=-0.7}
   }
   n2 = "Transform (Modify) Bone - Bone: dyn_mask_main"(ComponentPose=@n1, Alpha=Get RigidBodyDisabled() -> (RigidBodyDisabled:float), Translation="0,0,2", Rotation="0,0,0", Scale="1,1,1") {
    rotationMode=BMM_Additive
    rotationSpace=BCS_BoneSpace
    translationMode=BMM_Additive
    translationSpace=BCS_BoneSpace
    alphaBoolBlend {blendOption=Linear}
    boneToModify {boneName=dyn_mask_main}
    translation {z=2}
   }
   n3 = "Constraint - dyn_mask_L_deform (R)"(ComponentPose=@n2, Alpha=0.1, ConstraintWeights_0=1) {
    alphaBoolBlend {blendOption=Linear}
    boneToModify {boneName=dyn_mask_L_deform}
    constraintSetup [{transformType=Rotation, targetBone {boneName=clavicle_l}}]
    constraintWeights [1]
   }
   n4 = "Constraint - dyn_mask_R_deform (R)"(ComponentPose=@n3, Alpha=0.1, ConstraintWeights_0=1) {
    alphaBoolBlend {blendOption=Linear}
    boneToModify {boneName=dyn_mask_R_deform}
    constraintSetup [{transformType=Rotation, targetBone {boneName=clavicle_r}}]
    constraintWeights [1]
   }
   n5 = Get Gravity Override()
   n6 = RigidBody!(ComponentPose=@n4, Alpha, OverrideWorldGravity:Vector=@n5, ExternalForce) -> (Pose) {
    bClampLinearTranslationLimitToRefPose=true
    bOverrideWorldGravity=true
    overridePhysicsAsset=&/M_MED_EmeraldGlassGreen_Physics
    simulationSpace=BaseBoneSpace
    alphaBoolBlend {blendOption=Linear}
    baseBoneRef {boneName=spine_04}
    componentLinearAccScale=<0.3 0.3 0.3>
    componentLinearVelScale=<0.4 0.4 0.4>
    overrideWorldGravity {z=-980}
   }
   n7 = RigidBody!(
    ComponentPose=@n6
    bAlphaBoolEnabled:bool=Get backbling_NotEquipped() -> (backbling_NotEquipped:bool)
    OverrideWorldGravity=@n5
    ExternalForce
   ) -> (Pose) {
    alphaInputType=Bool
    bClampLinearTranslationLimitToRefPose=true
    bOverrideWorldGravity=true
    overridePhysicsAsset=&/M_MED_EmeraldGlassGreen_Physics_Cowl
    simulationSpace=BaseBoneSpace
    alphaBoolBlend {blendOption=Linear}
    baseBoneRef {boneName=spine_04}
    componentLinearAccScale=<0.3 0.3 0.3>
    componentLinearVelScale=<0.4 0.4 0.4>
    overrideWorldGravity {z=-980}
   }
   n8 = "Constraint - dyn_mask_L_deform (R)"(ComponentPose=@n7, Alpha=0.3, ConstraintWeights_0=1) {
    alphaBoolBlend {blendOption=Linear}
    boneToModify {boneName=dyn_mask_L_deform}
    constraintSetup [{transformType=Rotation, targetBone {boneName=clavicle_l}}]
    constraintWeights [1]
   }
   n9 = "Constraint - dyn_mask_R_deform (R)"(ComponentPose=@n8, Alpha=0.45, ConstraintWeights_0=1) {
    alphaBoolBlend {blendOption=Linear}
    boneToModify {boneName=dyn_mask_R_deform}
    constraintSetup [{transformType=Rotation, targetBone {boneName=clavicle_r}}]
    constraintWeights [1]
   }
   n10 = Get emeraldGlassPushFloat() -> (emeraldGlassPushFloat:double)
   n11 = "Transform (Modify) Bone - Bone: dyn_mask_main"(ComponentPose=@n9, Alpha=@n10, Translation="2.208,-3.839,1.359", Rotation="-16.379,-8.536,-13.608", Scale="1,1,1") {
    rotationMode=BMM_Additive
    rotationSpace=BCS_BoneSpace
    translationMode=BMM_Additive
    translationSpace=BCS_BoneSpace
    alphaBoolBlend {blendOption=Linear}
    alphaScaleBiasClamp {bInterpResult=true, interpSpeedDecreasing=5, interpSpeedIncreasing=0.5}
    boneToModify {boneName=dyn_mask_main}
    rotation=<rot -16.379 -8.536 -13.608>
    translation=<2.208 -3.839 1.359>
   }
   n12 = "Transform (Modify) Bone - Bone: dyn_mask_L_deform"(ComponentPose=@n11, Alpha=@n10, Translation="2.208,-3.839,1.359", Rotation="-16.379,-8.536,-13.608", Scale="1,1,1") {
    rotationMode=BMM_Additive
    rotationSpace=BCS_BoneSpace
    translationMode=BMM_Additive
    translationSpace=BCS_BoneSpace
    alphaBoolBlend {blendOption=Linear}
    alphaScaleBiasClamp {bInterpResult=true, interpSpeedDecreasing=5, interpSpeedIncreasing=0.5}
    boneToModify {boneName=dyn_mask_L_deform}
    rotation=<rot -16.379 -8.536 -13.608>
    translation=<2.208 -3.839 1.359>
   }
   n13 = "Transform (Modify) Bone - Bone: dyn_mask_main"(
    ComponentPose=@n12
    Alpha=Get neckForwardRotatedAlpha() -> (neckForwardRotatedAlpha:double)
    Translation="0.8,0,0"
    Rotation="-10,0,0"
    Scale="1,1,1"
   ) {
    rotationMode=BMM_Additive
    rotationSpace=BCS_BoneSpace
    translationMode=BMM_Additive
    translationSpace=BCS_BoneSpace
    alphaBoolBlend {blendOption=Linear}
    alphaScaleBiasClamp {interpSpeedDecreasing=5, interpSpeedIncreasing=0.5}
    boneToModify {boneName=dyn_mask_main}
    rotation {pitch=-10}
    translation {x=0.8}
   }
   n14 = Anim Dynamics - Bone: dyn_R_foot_lace_A(ComponentPose=@n13, Alpha=1, ExternalForce="0,0,0") {
    angularDampingOverride=1
    angularSpringConstant=20000
    bAngularSpring=true
    bOverrideAngularDamping=true
    bUseGravityOverride=true
    gravityScale=0.35
    simulationSpace=RootRelative
    alphaBoolBlend {blendOption=Linear}
    boundBone {boneName=dyn_R_foot_lace_A}
    componentLinearAccScale=<1 1 1>
    componentLinearVelScale=<1 1 1>
    gravityOverride {z=-980}
    physicsBodyDefinitions [{
      boundBone {boneName=dyn_R_foot_lace_A}
      boxExtents=<5 5 5>
      constraintSetup {angularLimitsMax=<0 40 30>, angularLimitsMin=<0 -10 -30>, angularTarget=<0 0 0>}
      localJointOffset=<-1 0 0>
     }]
   }
   n15 = Anim Dynamics - Bone: dyn_R_foot_lace_B(ComponentPose=@n14, Alpha=1, ExternalForce="0,0,0") {
    angularDampingOverride=1
    angularSpringConstant=20000
    bAngularSpring=true
    bOverrideAngularDamping=true
    bUseGravityOverride=true
    gravityScale=0.35
    simulationSpace=RootRelative
    alphaBoolBlend {blendOption=Linear}
    boundBone {boneName=dyn_R_foot_lace_B}
    componentLinearAccScale=<1 1 1>
    componentLinearVelScale=<1 1 1>
    gravityOverride {z=-980}
    physicsBodyDefinitions [{
      boundBone {boneName=dyn_R_foot_lace_B}
      boxExtents=<5 5 5>
      constraintSetup {angularLimitsMax=<0 40 30>, angularLimitsMin=<0 -10 -30>, angularTarget=<0 0 0>}
      localJointOffset=<-1 0 0>
     }]
   }
   n16 = Anim Dynamics - Bone: dyn_R_foot_laceLoop_A(ComponentPose=@n15, Alpha=1, ExternalForce="0,0,0") {
    angularDampingOverride=1
    angularSpringConstant=20000
    bAngularSpring=true
    bOverrideAngularDamping=true
    bUseGravityOverride=true
    gravityScale=0.35
    simulationSpace=RootRelative
    alphaBoolBlend {blendOption=Linear}
    boundBone {boneName=dyn_R_foot_laceLoop_A}
    componentLinearAccScale=<1 1 1>
    componentLinearVelScale=<1 1 1>
    gravityOverride {z=-980}
    physicsBodyDefinitions [{
      boundBone {boneName=dyn_R_foot_laceLoop_A}
      boxExtents=<5 5 5>
      constraintSetup {angularLimitsMax=<0 20 40>, angularLimitsMin=<0 0 0>, angularTarget=<0 0 0>}
      localJointOffset=<-1 0 0>
     }]
   }
   n17 = Anim Dynamics - Bone: dyn_R_foot_laceLoop_B(ComponentPose=@n16, Alpha=1, ExternalForce="0,0,0") {
    angularDampingOverride=1
    angularSpringConstant=20000
    bAngularSpring=true
    bOverrideAngularDamping=true
    bUseGravityOverride=true
    gravityScale=0.35
    simulationSpace=RootRelative
    alphaBoolBlend {blendOption=Linear}
    boundBone {boneName=dyn_R_foot_laceLoop_B}
    componentLinearAccScale=<1 1 1>
    componentLinearVelScale=<1 1 1>
    gravityOverride {z=-980}
    physicsBodyDefinitions [{
      boundBone {boneName=dyn_R_foot_laceLoop_B}
      boxExtents=<5 5 5>
      constraintSetup {angularLimitsMax=<0 20 0>, angularLimitsMin=<0 0 -40>, angularTarget=<0 0 0>}
      localJointOffset=<-1 0 0>
     }]
   }
   n18 = Anim Dynamics - Bone: dyn_L_foot_lace_A(ComponentPose=@n17, Alpha=1, ExternalForce="0,0,0") {
    angularDampingOverride=1
    angularSpringConstant=20000
    bAngularSpring=true
    bOverrideAngularDamping=true
    bUseGravityOverride=true
    gravityScale=0.35
    simulationSpace=RootRelative
    alphaBoolBlend {blendOption=Linear}
    boundBone {boneName=dyn_L_foot_lace_A}
    componentLinearAccScale=<1 1 1>
    componentLinearVelScale=<1 1 1>
    gravityOverride {z=-980}
    physicsBodyDefinitions [{
      boundBone {boneName=dyn_L_foot_lace_A}
      boxExtents=<5 5 5>
      constraintSetup {angularLimitsMax=<0 40 30>, angularLimitsMin=<0 -10 -30>, angularTarget=<0 0 0>}
      localJointOffset=<1 0 0>
     }]
   }
   n19 = Anim Dynamics - Bone: dyn_L_foot_lace_B(ComponentPose=@n18, Alpha=1, ExternalForce="0,0,0") {
    angularDampingOverride=1
    angularSpringConstant=20000
    bAngularSpring=true
    bOverrideAngularDamping=true
    bUseGravityOverride=true
    gravityScale=0.35
    simulationSpace=RootRelative
    alphaBoolBlend {blendOption=Linear}
    boundBone {boneName=dyn_L_foot_lace_B}
    componentLinearAccScale=<1 1 1>
    componentLinearVelScale=<1 1 1>
    gravityOverride {z=-980}
    physicsBodyDefinitions [{
      boundBone {boneName=dyn_L_foot_lace_B}
      boxExtents=<5 5 5>
      constraintSetup {angularLimitsMax=<0 40 30>, angularLimitsMin=<0 -10 -30>, angularTarget=<0 0 0>}
      localJointOffset=<1 0 0>
     }]
   }
   n20 = Anim Dynamics - Bone: dyn_L_foot_laceLoop_A(ComponentPose=@n19, Alpha=1, ExternalForce="0,0,0") {
    angularDampingOverride=1
    angularSpringConstant=20000
    bAngularSpring=true
    bOverrideAngularDamping=true
    bUseGravityOverride=true
    gravityScale=0.35
    simulationSpace=RootRelative
    alphaBoolBlend {blendOption=Linear}
    boundBone {boneName=dyn_L_foot_laceLoop_A}
    componentLinearAccScale=<1 1 1>
    componentLinearVelScale=<1 1 1>
    gravityOverride {z=-980}
    physicsBodyDefinitions [{
      boundBone {boneName=dyn_L_foot_laceLoop_A}
      boxExtents=<5 5 5>
      constraintSetup {angularLimitsMax=<0 0 40>, angularLimitsMin=<0 -20 0>, angularTarget=<0 0 0>}
      localJointOffset=<1 0 0>
     }]
   }
   n21 = Anim Dynamics - Bone: dyn_L_foot_laceLoop_B(ComponentPose=@n20, Alpha=1, ExternalForce="0,0,0") {
    angularDampingOverride=1
    angularSpringConstant=20000
    bAngularSpring=true
    bOverrideAngularDamping=true
    bUseGravityOverride=true
    gravityScale=0.35
    simulationSpace=RootRelative
    alphaBoolBlend {blendOption=Linear}
    boundBone {boneName=dyn_L_foot_laceLoop_B}
    componentLinearAccScale=<1 1 1>
    componentLinearVelScale=<1 1 1>
    gravityOverride {z=-980}
    physicsBodyDefinitions [{
      boundBone {boneName=dyn_L_foot_laceLoop_B}
      boxExtents=<5 5 5>
      constraintSetup {angularLimitsMax=<0 0 0>, angularLimitsMin=<0 -20 -40>, angularTarget=<0 0 0>}
      localJointOffset=<1 0 0>
     }]
   }
   n22 = "Constraint - dyn_L_kneeElement (RR)"!(ComponentPose=@n21, Alpha=1, ConstraintWeights_0=0.5, ConstraintWeights_1:float=0.5) -> (Pose) {
    alphaBoolBlend {blendOption=Linear}
    boneToModify {boneName=dyn_L_kneeElement}
    constraintSetup [{transformType=Rotation, targetBone {boneName=thigh_l}}, {transformType=Rotation, targetBone {boneName=calf_l}}]
    constraintWeights [1,1]
   }
   n23 = "Constraint - dyn_R_kneeElement (RR)"!(ComponentPose=@n22, Alpha=1, ConstraintWeights_0=0.5, ConstraintWeights_1:float=0.5) -> (Pose) {
    alphaBoolBlend {blendOption=Linear}
    boneToModify {boneName=dyn_R_kneeElement}
    constraintSetup [{transformType=Rotation, targetBone {boneName=thigh_r}}, {transformType=Rotation, targetBone {boneName=calf_r}}]
    constraintWeights [1,1]
   }
   n24 = Get L_wristRotated() -> (L_wristRotated:double)
   n25 = "Transform (Modify) Bone - Bone: dyn_L_wristElement_handScale"!(ComponentPose=@n23, Alpha=@n24, Translation="0,0,0", Rotation="0,0,0") -> (Pose) {
    scaleMode=BMM_Additive
    scaleSpace=BCS_BoneSpace
    alphaBoolBlend {blendOption=Linear}
    boneToModify {boneName=dyn_L_wristElement_handScale}
    scale {x=0.8}
   }
   n26 = Get R_wristRotated() -> (R_wristRotated:double)
   n27 = "Transform (Modify) Bone - Bone: dyn_R_wristElement_handScale"!(ComponentPose=@n25, Alpha=@n26, Translation="0,0,0", Rotation="0,0,0") -> (Pose) {
    scaleMode=BMM_Additive
    scaleSpace=BCS_BoneSpace
    alphaBoolBlend {blendOption=Linear}
    boneToModify {boneName=dyn_R_wristElement_handScale}
    scale {x=0.8}
   }
   n28 = "Transform (Modify) Bone - Bone: dyn_L_wristElement"!(ComponentPose=@n27, Alpha=@n24, Translation="-1.5,0,0", Rotation="0,0,0") -> (Pose) {
    scaleMode=BMM_Additive
    scaleSpace=BCS_BoneSpace
    translationMode=BMM_Additive
    translationSpace=BCS_BoneSpace
    alphaBoolBlend {blendOption=Linear}
    boneToModify {boneName=dyn_L_wristElement}
    scale {x=0.7}
    translation {x=-1.5}
   }
   n29 = "Transform (Modify) Bone - Bone: dyn_R_wristElement"!(ComponentPose=@n28, Alpha=@n26, Translation="1.5,0,0", Rotation="0,0,0") -> (Pose) {
    translationMode=BMM_Additive
    translationSpace=BCS_BoneSpace
    alphaBoolBlend {blendOption=Linear}
    boneToModify {boneName=dyn_R_wristElement}
    translation {x=1.5}
   }
   n30 = Anim Dynamics - Bone: dyn_L_front_pocket(ComponentPose=@n29, Alpha=1, ExternalForce="0,0,0") {
    angularDampingOverride=1
    angularSpringConstant=100
    bAngularSpring=true
    bOverrideAngularDamping=true
    bUseGravityOverride=true
    gravityScale=0.35
    simulationSpace=RootRelative
    alphaBoolBlend {blendOption=Linear}
    boundBone {boneName=dyn_L_front_pocket}
    componentLinearAccScale=<1 1 1>
    componentLinearVelScale=<1 1 1>
    gravityOverride {z=-980}
    physicsBodyDefinitions [{
      boundBone {boneName=dyn_L_front_pocket}
      boxExtents=<30 30 30>
      constraintSetup {angularLimitsMax=<10 0 5>, angularLimitsMin=<-10 -20 -5>, angularTarget=<1 0 0>}
      localJointOffset=<1 0 -3>
     }]
   }
   n31 = Anim Dynamics - Bone: dyn_L_back_pocket(ComponentPose=@n30, Alpha=1, ExternalForce="0,0,0") {
    angularDampingOverride=1
    angularSpringConstant=100
    bAngularSpring=true
    bOverrideAngularDamping=true
    bUseGravityOverride=true
    gravityScale=0.35
    simulationSpace=RootRelative
    alphaBoolBlend {blendOption=Linear}
    boundBone {boneName=dyn_L_back_pocket}
    componentLinearAccScale=<1 1 1>
    componentLinearVelScale=<1 1 1>
    gravityOverride {z=-980}
    physicsBodyDefinitions [{
      boundBone {boneName=dyn_L_back_pocket}
      boxExtents=<45 45 45>
      constraintSetup {angularLimitsMax=<10 0 5>, angularLimitsMin=<-10 -10 -5>, angularTarget=<1 0 0>}
      localJointOffset=<1 0 -3>
     }]
   }
   n32 = Anim Dynamics - Bone: dyn_R_front_pocket(ComponentPose=@n31, Alpha=1, ExternalForce="0,0,0") {
    angularDampingOverride=1
    angularSpringConstant=100
    bAngularSpring=true
    bOverrideAngularDamping=true
    bUseGravityOverride=true
    gravityScale=0.35
    simulationSpace=RootRelative
    alphaBoolBlend {blendOption=Linear}
    boundBone {boneName=dyn_R_front_pocket}
    componentLinearAccScale=<1 1 1>
    componentLinearVelScale=<1 1 1>
    gravityOverride {z=-980}
    physicsBodyDefinitions [{
      boundBone {boneName=dyn_R_front_pocket}
      boxExtents=<45 45 45>
      constraintSetup {angularLimitsMax=<10 0 5>, angularLimitsMin=<-10 -10 -5>, angularTarget=<1 0 0>}
      localJointOffset=<1 0 -3>
     }]
   }
   n33 = Anim Dynamics - Bone: dyn_R_back_pocket(ComponentPose=@n32, Alpha=1, ExternalForce="0,0,0") {
    angularDampingOverride=1
    angularSpringConstant=100
    bAngularSpring=true
    bOverrideAngularDamping=true
    bUseGravityOverride=true
    gravityScale=0.35
    simulationSpace=RootRelative
    alphaBoolBlend {blendOption=Linear}
    boundBone {boneName=dyn_R_back_pocket}
    componentLinearAccScale=<1 1 1>
    componentLinearVelScale=<1 1 1>
    gravityOverride {z=-980}
    physicsBodyDefinitions [{
      boundBone {boneName=dyn_R_back_pocket}
      boxExtents=<50 50 50>
      constraintSetup {angularLimitsMax=<10 0 5>, angularLimitsMin=<-10 -25 -5>, angularTarget=<1 0 0>}
      localJointOffset=<1 0 -3>
     }]
   }
   n34 = Get L_thighRotatedOut() -> (L_thighRotatedOut:double)
   n35 = "Transform (Modify) Bone - Bone: dyn_L_front_pocket"!(ComponentPose=@n33, Alpha=@n34, Scale="1,1,1") -> (Pose) {
    rotationMode=BMM_Additive
    translationMode=BMM_Additive
    translationSpace=BCS_BoneSpace
    alphaBoolBlend {blendOption=Linear}
    boneToModify {boneName=dyn_L_front_pocket}
    rotation=<rot 24.2413 -18.9646 -14.9125>
    translation {y=-1.1}
   }
   n36 = "Transform (Modify) Bone - Bone: dyn_L_back_pocket"!(ComponentPose=@n35, Alpha=@n34, Scale="1,1,1") -> (Pose) {
    rotationMode=BMM_Additive
    translationMode=BMM_Additive
    translationSpace=BCS_BoneSpace
    alphaBoolBlend {blendOption=Linear}
    boneToModify {boneName=dyn_L_back_pocket}
    rotation=<rot 3.1569 -3.5203 0.458>
    translation {y=-0.6}
   }
   n37 = "Transform (Modify) Bone - Bone: dyn_R_front_pocket"!(ComponentPose=@n36, Alpha=Get R_thighRotated() -> (R_thighRotated:double), Scale="1,1,1") -> (Pose) {
    rotationMode=BMM_Additive
    translationMode=BMM_Additive
    translationSpace=BCS_BoneSpace
    alphaBoolBlend {blendOption=Linear}
    alphaScaleBiasClamp {bInterpResult=true, interpSpeedDecreasing=20, interpSpeedIncreasing=20}
    boneToModify {boneName=dyn_R_front_pocket}
    rotation=<rot -21.7153 -16.9838 25.268>
   }
   Output Pose(Result=Component To Local(ComponentPose=@n37)) {layerGroup=None}
  ]}
 EventGraph {nodes [
   Event BlueprintUpdateAnimation()
   > n1 = Sequence!(execute) -> (then_0, then_1, then_2:Exec)
   n2 = GetOwnerSkeletalMeshComponent()
   n3 = BreakRotator(InRot=BreakTransform(InTransform=GetDeltaTransformFromRefPose(self=@n2, BoneName=hand_l, BaseName=lowerarm_l)).Rotation)
   > Set L_wristRotated(L_wristRotated:double=float + float!(
     A=MapRangeClamped(Value=@n3.Yaw, InRangeA=0, InRangeB=60)
     B=MapRangeClamped(Value=@n3.Pitch, InRangeA=0, InRangeB=60)
     C:double=MapRangeClamped(Value=@n3.Pitch, InRangeA=0, InRangeB=-60)
    ) -> (ReturnValue)) -> (Output_Get:double)
   n4 = BreakRotator(InRot=BreakTransform(InTransform=GetDeltaTransformFromRefPose(self=@n2, BoneName=hand_r, BaseName=lowerarm_r)).Rotation)
   > Set R_wristRotated(R_wristRotated:double=float + float(
     A=MapRangeClamped(Value=@n4.Yaw, InRangeA=0, InRangeB=60)
     B=MapRangeClamped(Value=@n4.Pitch, InRangeA=0, InRangeB=60)
     C=MapRangeClamped(Value=@n4.Pitch, InRangeA=0, InRangeB=-60)
    )) -> (Output_Get:double)
   n5 = GetOwnerSkeletalMeshComponent()
   n6 = BreakTransform(InTransform=GetDeltaTransformFromRefPose(self=@n5, BoneName=thigh_l, BaseName=pelvis))
   > Set L_thighRotatedOut(L_thighRotatedOut:double=MapRangeClamped(Value=BreakRotator(InRot=@n6.Rotation).Pitch, InRangeA=10, InRangeB=40)) -> (Output_Get:double)
   n7 = BreakTransform(InTransform=GetDeltaTransformFromRefPose(self=@n5, BoneName=thigh_r, BaseName=pelvis))
   > Set R_thighRotated(R_thighRotated:double=MapRangeClamped(Value=BreakRotator(InRot=@n7.Rotation).Yaw, InRangeA=-60, InRangeB=-100)) -> (Output_Get:double)
   n8 = BreakTransform(InTransform=GetDeltaTransformFromRefPose(self=@n5, BoneName=neck_01, BaseName=spine_05))
   > Set neckForwardRotatedAlpha(neckForwardRotatedAlpha:double=MapRangeClamped(Value=BreakRotator(InRot=@n8.Rotation).Yaw, InRangeA=0, InRangeB=15)) -> (Output_Get:double)
   n9 = Gravity Override State Function DEPRECATED(
    execute=@n1.then_2
    CharacterPartAnimInstance=Self-Reference()
    GameTimeInSeconds=GetGameTimeInSeconds()
    GravityOverrideParams=Get Gravity Override Params()
   )
   > n10 = Set Gravity Override(Gravity Override=@n9)
   > Set Gravity Override(Gravity Override=SelectVector(A="0, 0,-4000", B=@n10, bPickA=Get backbling_equipped() -> (backbling_equipped:bool)))
   n11 = Get bRigidBodyEnabled()
   > Set RigidBodyDisabled(RigidBodyDisabled:float="To Float (Boolean)"(InBool=NOT Boolean(A=@n11))) -> (Output_Get:float)
   > n12 = Cast To FortPlayerPawn(Object=GetOwningActor())
   > Set As Fort Player Pawn(As Fort Player Pawn:Fort Player Pawn Object Reference=@n12) -> (Output_Get:Fort Player Pawn Object Reference)
   > n13 = Get!(execute:Exec) -> (then:Exec, else:Exec, As Fort Player Pawn:Fort Player Pawn Object Reference)
   n14 = Get bIsPlayingEmote()
   n15 = HasMatchingGameplayTag!(self:Gameplay Tag Asset Interface Interface=@n13, TagToCheck:Gameplay Tag Structure=(TagName="Granted.Athena.EmeraldGlassPush.Equipped")) -> (ReturnValue:bool)
   n16 = IsAnyMontagePlaying!(self:Anim Instance Object Reference=GetOwnerFortPlayerAnimInstance!(self:Custom Character Part Anim Instance Object Reference) -> (ReturnValue:Fort Player Anim Instance Object Reference)) -> (ReturnValue:bool)
   > n17 = "Set emeraldGlassPush!"("emeraldGlassPush!":bool=AND Boolean(A=AND Boolean(A=@n15, B=@n16), B=NOT Boolean(A=@n14))) -> (Output_Get:bool)
   > Set emeraldGlassPushFloat(emeraldGlassPushFloat:double="To Float (Boolean)"(InBool=@n17)) -> (Output_Get:double)
   n18 = CheckForBackbling()
   n19 = Set backbling_equipped(execute=@n1.then_1, backbling_equipped:bool=@n18) -> (Output_Get:bool)
   > Set backbling_NotEquipped(backbling_NotEquipped:bool=NOT Boolean(A=@n19)) -> (Output_Get:bool)
   > Set notCape(notCape:bool=NOT Boolean(A=@n18.bIsBackblingACape)) -> (Output_Get:bool)
  ]}
]
variables [
 As Fort Player Pawn {type=FortPlayerPawn}
 backbling_equipped {type=boolean}
 backbling_NotEquipped {type=boolean}
 Current Weapon {type=FortWeapon, default=null}
 "emeraldGlassPush!" {type=boolean}
 emeraldGlassPushFloat {type=number}
 Gravity Override {type=Vector}
 Gravity Override Params {type=GravityOverrideParamsStruct, default {onGround_Falling {gravityOverride=<0 0 -300>}}}
 L_thighRotatedOut {type=number}
 L_wristRotated {type=number}
 neckForwardRotatedAlpha {type=number}
 notCape {type=boolean}
 R_thighRotated {type=number}
 R_wristRotated {type=number}
 Recently Fired {type=boolean}
 RigidBodyDisabled {type=number}
]
